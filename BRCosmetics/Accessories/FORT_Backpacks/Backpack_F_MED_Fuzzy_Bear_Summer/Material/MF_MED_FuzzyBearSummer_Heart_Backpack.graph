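# Material function for the Fuzzy Bear Summer backpack's glowing glass heart, with default, take-damage and open-chest looks.
# It outputs black base color, a normal map and noise-varied specular and roughness, putting color into emissive.
# Fixed 2x2 flipbook cells supply heart-state art blended through HeartShapeColor, while a noise-distorted SphereMask driven by HeartFxMask forms a transition edge.
# View-dependent glitter, time-of-day-tinted cubemap reflections and Fresnel glows add faked glass shading.
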MaterialFunction /BRCosmetics/Accessories/FORT_Backpacks/Backpack_F_MED_Fuzzy_Bear_Summer/Material/MF_MED_FuzzyBearSummer_Heart_Backpack
graph {nodes [
  n1 = TextureSample(UVs=TextureCoordinate()) {samplerSource=SSM_Wrap_WorldGroupSettings, samplerType=SAMPLERTYPE_Masks, texture=/Game/Global/Textures/Default/Noise/T_Metal_Noise}
  n2 = Add(A=TextureCoordinate(), B=Multiply(A=@n1.R) {constB=0.1})
  n3 = AppendVector(A=Constant() {r=0.51}, B=Constant() {r=0.5})
  n4 = LinearInterpolate(Alpha=ScalarParameter() {parameterName=HeartFxMask})
  n5 = SphereMask(A=@n2, B=@n3, Radius=@n4, Hardness=Constant() {r=0.95})
  n6 = SphereMask(A=@n2, B=@n3, Radius=Subtract(A=@n4, B=Constant() {r=0.044}) {constB=-0.327}, Hardness=Constant() {r=0.8519})
  n7 = Multiply(A=Saturate(None=Subtract(A=@n5, B=@n6)), B=Constant3Vector() {constant=<rgba 0.1 0 0.0346 1>})
  n8 = Constant() {r=80}
  n9 = Multiply(A=NamedRerouteUsage()) {constB=0.2}
  n10 = Fresnel(ExponentIn=ScalarParameter() {defaultValue=0.07, parameterName=Param_3}) {baseReflectFraction=0}
  n11 = Multiply(A=Power(Base=OneMinus(None=@n10)), B=NamedRerouteUsage())
  n12 = TextureSampleParameter2D(UVs=Multiply(
    A=TextureCoordinate()
    B=ScalarParameter() {defaultValue=3, group=Glitter, parameterName=Detail Tiling, sliderMax=64, sliderMin=8, sortPriority=2}
   ) {constB=3}) {
   group=Glitter
   parameterName=Glitter Direction Mask
   samplerType=SAMPLERTYPE_Masks
   sortPriority=0
   texture=/Game/Weapons/WeaponSkins/Textures/T_GlitterDirectionMask
  }
  n13 = LinearInterpolate(A=Constant3Vector() {constant=<rgba 6 5.4 5.9753 0>}, B=Constant3Vector() {constant=<rgba 1 0.9 0.9959 0>}, Alpha=@n12.R)
  n14 = NamedRerouteUsage()
  n15 = NamedRerouteUsage()
  n16 = NamedRerouteUsage()
  n17 = ScalarParameter() {parameterName=HeartShapeColor}
  n18 = LinearInterpolate(
   A=@n14
   B=NamedRerouteDeclaration(None=MaterialFunctionCall(A=@n14, B=@n15, C=@n16, Alpha=@n17) {materialFunction=&Lerp_3Color}.Result) {name=BaseColor, nodeColor=<rgba 0.2891 1 0 1>}
   Alpha=NamedRerouteUsage()
  )
  n19 = NamedRerouteUsage()
  n20 = Multiply(A=Add(A=Multiply(A=@n11, B=@n13), B=Multiply(A=@n18, B=@n19)), B=ScalarParameter() {defaultValue=0.38, parameterName=GlowStrength})
  n21 = TextureSample() {
   samplerType=SAMPLERTYPE_Normal
   texture=/BRCosmetics/Accessories/FORT_Backpacks/Backpack_F_MED_Fuzzy_Bear_Summer/Texture/T_F_MED_Fuzzy_Bear_Summer_Backpack_Heart_N
  }
  n22 = SetMaterialAttributes!(
   MaterialAttributes=GetMaterialAttributes(None=FunctionInput() {bUsePreviewValueAsDefault=true, inputType=FunctionInput_MaterialAttributes}).MaterialAttributes
   BaseColor=Constant3Vector() {constant=<rgba 0 0 0 1>}
   Metallic
   Specular=NamedRerouteUsage()
   Roughness=NamedRerouteUsage()
   EmissiveColor=Add(A=Multiply(A=@n7, B=@n8) {constB=150}, B=Add(A=@n9, B=@n20))
   Normal=@n21.RGB
   AmbientOcclusion
  ) {attributeSetTypes [
    #1
    #2
    #3
    #4
    #5
    #6
    #7
   ]}
  FunctionOutput(None=@n22)
  n23 = ScalarParameter() {defaultValue=0.95, group=Glitter, parameterName=Max Glitter, sliderMax=1, sortPriority=8}
  n24 = CameraVectorWS() {desc=ViewDir}
  n25 = Transform(None=Constant3Vector() {constant=<rgba 0 0 -1 0>}) {desc=Camera Vector Forward, transformSourceType=TRANSFORMSOURCE_Camera}
  n26 = TextureSampleParameter2D(UVs=Multiply(
    A=TextureCoordinate()
    B=ScalarParameter() {defaultValue=3, group=Glitter, parameterName=Detail Tiling, sliderMax=64, sliderMin=8, sortPriority=2}
   ) {constB=3}) {
   group=Glitter
   parameterName=Glitter Direction Mask
   samplerType=SAMPLERTYPE_Masks
   sortPriority=0
   texture=&T_GlitterDirectionMask
  }
  n27 = TextureSampleParameter2D(UVs=Multiply(A=TextureCoordinate(), B=Constant() {r=6}) {constB=3}) {
   group=Glitter
   parameterName=Glitter Direction Mask
   samplerType=SAMPLERTYPE_Masks
   sortPriority=0
   texture=&T_GlitterDirectionMask
  }
  n28 = TextureSample(UVs=TextureCoordinate() {uTiling=3, vTiling=3}) {samplerSource=SSM_Wrap_WorldGroupSettings, samplerType=SAMPLERTYPE_Masks, texture=&T_Metal_Noise}
  n29 = ComponentMask(None=LinearInterpolate(A=@n26.RGBA, B=@n27.RGBA, Alpha=Power(Base=@n28.R) {constExponent=3})) {b=true, g=true, r=true}
  n30 = Multiply(
   A=Add(A=Multiply(A=@n29) {constB=2}) {constB=-1}
   B=ScalarParameter() {defaultValue=0.928, group=Glitter, parameterName=Normal Noise Intensity, sliderMax=2, sortPriority=6}
  ) {constB=0.5}
  n31 = VertexNormalWS()
  n32 = DotProduct(A=Normalize(VectorInput=Add(A=@n24, B=@n25)) {desc="L + V  average"}, B=Normalize(VectorInput=Add(A=@n30, B=@n31)) {desc=GlitterDirection})
  NamedRerouteDeclaration(None=Power(Base=SmoothStep(Min=@n23, Value=@n32), Exp=ScalarParameter() {defaultValue=3, parameterName=Param_2}) {constExponent=3}) {name=Flake, nodeColor=<rgba 1 0.375 0 1>}
  n33 = TextureSample() {
   samplerType=SAMPLERTYPE_Normal
   texture=&T_F_MED_Fuzzy_Bear_Summer_Backpack_Heart_N
  }
  n34 = Normalize(VectorInput=Multiply(A=@n33.RGB, B=VectorParameter() {parameterName=Glass_NormalIntensity, defaultValue=<rgba 0.02 0.02 1 0>}.RGB))
  n35 = TextureSample(UVs=LinearInterpolate(A=ReflectionVectorWS(CustomWorldNormal=@n34), B=ReflectionVectorWS()) {constAlpha=0.2}) {samplerType=SAMPLERTYPE_LinearColor, texture=/BRCosmetics/Accessories/FORT_Backpacks/Backpack_F_MED_Fuzzy_Bear_Summer/Texture/AthenaCubemap2}
  n36 = Multiply(A=MaterialFunctionCall() {materialFunction=&Color_Time_of_Day}.Result, B=@n35.RGB)
  n37 = Fresnel(ExponentIn=ScalarParameter() {defaultValue=0.8, parameterName=Glass_CubemapFresnel})
  n38 = Multiply(
   A=Power(Base=Multiply(A=@n36, B=@n37), Exp=ScalarParameter() {defaultValue=1.5, parameterName=Glass_CubemapContrast})
   B=ScalarParameter() {defaultValue=1, parameterName=Glass_CubemapIntensity}
  )
  n39 = NamedRerouteUsage()
  NamedRerouteDeclaration(None=LinearInterpolate(A=@n38, B=Multiply(A=@n38, B=@n39), Alpha=ScalarParameter() {defaultValue=0.5, parameterName=Glass_CubemapColorTint})) {name=Cubemap, nodeColor=<rgba 0 1 0.2031 1>}
  n40 = MaterialFunctionCall() {materialFunction=&VectorToRadialValue}
  n41 = TextureSample(UVs=Multiply(A=Panner(Coordinate=@n40.Radial Coordinates) {speedX=0.01, speedY=-0.3})) {texture=/Game/Global/Textures/Default/Noise/Noise}
  n42 = Constant() {r=2}
  n43 = StaticComponentMaskParameter(None=MaterialFunctionCall(
    "Animation  Phase (0-1)"=Constant() {r=0.75}
    Number of Rows=@n42
    Number of Columns=@n42
    Texture=TextureObject() {texture=/BRCosmetics/Accessories/FORT_Backpacks/Backpack_F_MED_Fuzzy_Bear_Summer/Texture/T_F_MED_Fuzzy_Bear_Summer_Backpack_FX2}
   ) {materialFunction=&FlipBook}.Result) {defaultG=true}
  n44 = LinearInterpolate(
   A=Constant()
   B=Multiply(A=Power(Base=@n41.R) {constExponent=3}, B=Constant3Vector() {constant=<rgba 200 48.2595 160.6579 0>})
   Alpha=Power(Base=Multiply(A=@n43)) {constExponent=1.2}
  )
  n45 = Constant() {r=0.2}
  n46 = Constant() {r=2}
  n47 = TextureObject() {texture=&T_F_MED_Fuzzy_Bear_Summer_Backpack_FX2}
  n48 = Multiply(
   A=StaticComponentMaskParameter(None=MaterialFunctionCall("Animation  Phase (0-1)"=@n45, Number of Rows=@n46, Number of Columns=@n46, Texture=@n47) {materialFunction=&FlipBook}.Result) {defaultG=true}
   B=Constant3Vector() {constant=<rgba 1 0.2413 0.8033 0>}
  )
  NamedRerouteDeclaration(None=Add(A=Max(A=@n44, B=@n48), B=VectorParameter() {defaultValue=<rgba 0.1 0 0.0782 1>}.RGB)) {name=default color, nodeColor=<rgba 1 0.2109 0 1>}
  n49 = Constant() {r=4}
  n50 = Fresnel(ExponentIn=Constant() {r=-0.6}) {baseReflectFraction=0}
  n51 = ScalarParameter() {defaultValue=1.2, parameterName=OuterGlows_Amount}
  NamedRerouteDeclaration(None=Add(
    A=MaterialFunctionCall(Exponent=@n49) {materialFunction=&MF_Fresnel_FX}.Fresnel
    B=Multiply(A=@n50, B=@n51)
   )) {name=Fresnel, nodeColor=<rgba 0.9453 0 1 1>}
  n52 = ScalarParameter() {defaultValue=0.2, parameterName=Specular Min}
  n53 = ScalarParameter() {defaultValue=1, parameterName=Specular Max}
  n54 = TextureSample() {samplerSource=SSM_Wrap_WorldGroupSettings, samplerType=SAMPLERTYPE_Masks, texture=&T_Metal_Noise}
  NamedRerouteDeclaration(None=LinearInterpolate(A=LinearInterpolate(A=@n52, B=@n53, Alpha=@n54.R), B=@n53, Alpha=NamedRerouteUsage())) {name=Spec, nodeColor=<rgba 1 0 0.6094 1>}
  NamedRerouteDeclaration(None=@n5) {name=FXMask, nodeColor=<rgba 0.1016 1 0 1>}
  n55 = TextureSample(UVs=Multiply(A=TextureCoordinate(), B=ScalarParameter() {defaultValue=0.75, parameterName=Rough Noise Scale})) {samplerSource=SSM_Wrap_WorldGroupSettings, samplerType=SAMPLERTYPE_Masks, texture=&T_Metal_Noise}
  NamedRerouteDeclaration(None=LinearInterpolate(
    A=ScalarParameter() {parameterName=Roughness Min}
    B=ScalarParameter() {defaultValue=0.35, parameterName=Roughness Max}
    Alpha=Power(Base=@n55.R) {constExponent=2.5}
   )) {name=Roughness, nodeColor=<rgba 0.8281 0 1 1>}
  n56 = Constant() {r=2}
  n57 = StaticComponentMaskParameter(None=MaterialFunctionCall(
    "Animation  Phase (0-1)"=Constant() {r=0.4}
    Number of Rows=@n56
    Number of Columns=@n56
    Texture=TextureObject() {texture=&T_F_MED_Fuzzy_Bear_Summer_Backpack_FX2}
   ) {materialFunction=&FlipBook}.Result) {defaultG=true}
  n58 = Constant3Vector() {constant=<rgba 0.3 0.3 0.3 0>}
  NamedRerouteDeclaration(None=Add(A=Multiply(A=@n57, B=@n58), B=VectorParameter() {parameterName=Param_1, defaultValue=<rgba 0.01 0.01 0.01 1>}.RGB)) {name=take damage, nodeColor=<rgba 1 0.6094 0 1>}
  NamedRerouteDeclaration(None=@n50) {name=Rim, nodeColor=<rgba 0.3359 1 0 1>}
  n59 = Constant() {r=2}
  n60 = StaticComponentMaskParameter(None=MaterialFunctionCall(
    "Animation  Phase (0-1)"=Constant() {r=0.5}
    Number of Rows=@n59
    Number of Columns=@n59
    Texture=TextureObject() {texture=&T_F_MED_Fuzzy_Bear_Summer_Backpack_FX2}
   ) {materialFunction=&FlipBook}.Result) {defaultG=true}
  n61 = Constant3Vector() {constant=<rgba 1 0.2253 0.0045 0>}
  NamedRerouteDeclaration(None=Add(A=Multiply(A=@n60, B=@n61), B=VectorParameter() {parameterName=Param_2, defaultValue=<rgba 0.15 0 0.0237 1>}.RGB)) {name=OpenChest, nodeColor=<rgba 1 0 0.9609 1>}
 ]}
