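# Material function for the Fuzzy Bear summer backpack cosmetic that switches mesh parts and surface look by emote state.
# An emote value, built from two scalar parameters, is matched against vertex colors to collapse hidden parts onto the actor position via World Position Offset.
# State 5 adds a sine-driven scale pulse and conditional base color, normal, roughness and metallic changes, apparently eye-masked.
# The FxMask texture supplies the eye mask and two-color emissive.
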MaterialFunction /BRCosmetics/Packages/Fortress_SharedMaterials/Base_Material_Functions/MF_FuzzyBearSummerBackpack
graph {nodes [
  n1 = GetMaterialAttributes(None=FunctionInput() {inputType=FunctionInput_MaterialAttributes}) {attributeGetTypes [
    #1
    #2
    #3
    #4
    #5
   ]}
  n2 = Subtract(A=ActorPositionWS(), B=WorldPosition().XYZ)
  n3 = LinearInterpolate(B=ScalarParameter() {parameterName=Bear summer Emote visibily}, Alpha=ScalarParameter() {parameterName=BearSummer_gametime})
  n4 = Saturate(None=Subtract(A=Divide(A=@n3) {constB=10}) {constB=0.55})
  n5 = Constant()
  n6 = Constant() {r=1}
  n7 = Constant() {r=1}
  n8 = If(A=@n3, B=Constant() {r=10}, "A > B"=@n7, "A == B"=Constant(), "A < B"=@n7)
  n9 = VertexColor()
  n10 = Constant()
  n11 = If(A=Ceil(None=Multiply(A=@n9.R) {constB=10}), B=@n3, "A > B"=@n10, "A == B"=Constant() {r=1}, "A < B"=@n10)
  n12 = Saturate(None=Add(A=If(A=@n9.R, B=@n4, "A > B"=@n5, "A == B"=@n6, "A < B"=@n8), B=Add(A=@n11, B=@n9.G)))
  n13 = Subtract(A=Add(A=ActorPositionWS(), B=Constant3Vector() {constant=<rgba 0 0 0 1>}), B=WorldPosition().XYZ)
  n14 = Constant() {r=1}
  n15 = Time()
  n16 = If(
   A=@n3
   B=Constant() {r=5}
   "A > B"=@n14
   "A == B"=LinearInterpolate(
    A=Constant() {r=1.01}
    Alpha=MaterialFunctionCall(Value=@n15) {materialFunction=&LinearSine}.Linear Sine
   ) {constA=1.1}
   "A < B"=@n14
  )
  n17 = SetMaterialAttributes!(
   MaterialAttributes=@n1.MaterialAttributes
   WorldPositionOffset=LinearInterpolate(
    A=LinearInterpolate(A=@n2, Alpha=@n12) {constA=1, constB=0, desc="deactivated = 0"}
    B=LinearInterpolate(A=@n13, Alpha=@n16) {constB=0}
    Alpha=@n11
   )
   EmissiveColor=NamedRerouteUsage()
   BaseColor=NamedRerouteUsage()
   Normal=NamedRerouteUsage()
   Roughness=NamedRerouteUsage()
   Metallic=NamedRerouteUsage()
  ) {attributeSetTypes [
    #6
    #1
    #2
    #3
    #4
    #5
   ]}
  FunctionOutput(None=@n17)
  n18 = Constant3Vector() {constant=<rgba 5 0.5 2.2812 1>}
  n19 = TextureSampleParameter2D() {group=FuzzyBearSummer, parameterName=FxMask}
  n20 = Constant3Vector() {constant=<rgba 3 0.3373 0 1>}
  NamedRerouteDeclaration(None=Max(A=Multiply(A=@n18, B=@n19.B), B=Multiply(A=@n19.R, B=@n20))) {name=Emisssive, nodeColor=<rgba 0 1 0.6719 1>}
  NamedRerouteDeclaration(None=Saturate(None=Multiply(A=@n19.A) {constB=5})) {name=EyeMask, nodeColor=<rgba 0 0.5781 1 1>}
  n21 = NamedRerouteUsage()
  n22 = NamedRerouteUsage()
  n23 = VectorParameter() {defaultValue=<rgba 0.2122 0.2122 0.2122 1>}
  n24 = NamedRerouteUsage()
  NamedRerouteDeclaration(None=If(A=NamedRerouteUsage(), B=Constant() {r=5}, "A > B"=@n21, "A == B"=LinearInterpolate(A=@n22, B=@n23.RGB, Alpha=@n24), "A < B"=@n21)) {name=Roughness, nodeColor=<rgba 0.0312 1 0 1>}
  NamedRerouteDeclaration(None=@n3) {name=EmotVisibily, nodeColor=<rgba 0.1953 0 1 1>}
  n25 = TextureSampleParameter2D() {group=FuzzyBearSummer, parameterName=FuzzyBearColor, texture=/Game/Global/Textures/Default/Blanks/T_Blank_LegalWhite_sRGBColor}
  n26 = NamedRerouteUsage()
  n27 = Constant3Vector() {constant=<rgba 0.8879 0.8963 0.8879 1>}
  n28 = NamedRerouteUsage()
  NamedRerouteDeclaration(None=If(A=NamedRerouteUsage(), B=Constant() {r=5}, "A > B"=@n25.RGB, "A == B"=LinearInterpolate(A=@n26, B=@n27, Alpha=@n28), "A < B"=@n25.RGB)) {name=Basecolor, nodeColor=<rgba 0 1 0.0859 1>}
  NamedRerouteDeclaration(None=@n1.BaseColor) {name=In basecolor, nodeColor=<rgba 1 0.75 0 1>}
  NamedRerouteDeclaration(None=@n1.Normal) {name=InNormal, nodeColor=<rgba 0.3359 0 1 1>}
  n29 = NamedRerouteUsage()
  n30 = LinearInterpolate(Alpha=NamedRerouteUsage())
  NamedRerouteDeclaration(None=If(
    A=NamedRerouteUsage()
    B=Constant() {r=5}
    "A > B"=@n29
    "A == B"=MaterialFunctionCall(Normal=@n29, Flatness=@n30) {materialFunction=&FlattenNormal}.Result
    "A < B"=@n29
   )) {name=normal, nodeColor=<rgba 0 1 0.2969 1>}
  NamedRerouteDeclaration(None=@n1.Metallic) {name=InMetallic, nodeColor=<rgba 1 0 0.8672 1>}
  n31 = NamedRerouteUsage()
  n32 = NamedRerouteUsage()
  n33 = Constant3Vector()
  n34 = NamedRerouteUsage()
  NamedRerouteDeclaration(None=If(A=NamedRerouteUsage(), B=Constant() {r=5}, "A > B"=@n31, "A == B"=LinearInterpolate(A=@n32, B=@n33, Alpha=@n34), "A < B"=@n31)) {name=metallic, nodeColor=<rgba 0 0.1562 1 1>}
  NamedRerouteDeclaration(None=@n1.Roughness) {name=InRoughness, nodeColor=<rgba 1 0.8906 0 1>}
 ]}
