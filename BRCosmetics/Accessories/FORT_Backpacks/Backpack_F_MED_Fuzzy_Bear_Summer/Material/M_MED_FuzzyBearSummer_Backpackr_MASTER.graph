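# Master character material for the Fuzzy Bear Summer backpack cosmetic, built from material attributes for skeletal meshes with Default Lit, Subsurface Profile and Cloth shading.
# The shared character PreFX function feeds an optional ClothFuzz layer, masked by the FX texture's red channel and toggled by the "Use Cloth fuzz" static switch.
# The result passes through the backpack-specific material function, then the shared character PostFX and cinematic PostFX functions.
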Material /BRCosmetics/Accessories/FORT_Backpacks/Backpack_F_MED_Fuzzy_Bear_Summer/Material/M_MED_FuzzyBearSummer_Backpackr_MASTER
graph {
 properties {
  bAutomaticallySetUsageInEditor=false
  bUseMaterialAttributes=true
  bUsedWithMeshDeformer=true
  bUsedWithMorphTargets=true
  bUsedWithSkeletalMesh=true
  pixelDepthOffsetMode=PDOM_Legacy
  shadingModel=MSM_FromMaterialExpression
  subsurfaceProfile=&SS_HeroSkin_02
  usedShadingModels=Default Lit | Subsurface Profile | Cloth
 }
 nodes [
  n1 = MaterialFunctionCall() {materialFunction=&BaseFN_CharacterFunction_PreFX}
  n2 = MaterialFunctionCall!(
   Alpha=TextureSample() {texture=/BRCosmetics/Accessories/FORT_Backpacks/Backpack_F_MED_Fuzzy_Bear_Summer/Texture/T_F_MED_Fuzzy_Bear_Summer_Backpack_FX}.R
   Gmap=@n1.Gmap
   Material=@n1.Gmap
  ) {materialFunction=/Game/Weapons/WeaponSkins/MaterialFunction/ClothFuzz}
  n3 = MaterialFunctionCall(
   AO=@n1.AO
   Distance Blend=@n1.AO
   Fresnel=@n1.AO
   Gmap=@n1.AO
   Pre Skinned Local Position=@n1.AO
   Material=MaterialFunctionCall!(In=StaticSwitchParameter(True=@n2.Result, False=@n1.Result) {group=Skin FX, parameterName=Use Cloth fuzz}) {materialFunction=/BRCosmetics/Packages/Fortress_SharedMaterials/Base_Material_Functions/MF_FuzzyBearSummerBackpack}.Result
  ) {materialFunction=&BaseFN_CharacterFunction_PostFX}
  MaterialFunctionCall(In=@n3.Result) {materialFunction=&CINE_MF_CharPostFX}
 ]
}
