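# Physics asset for the Fuzzy Bear Summer backpack cosmetic, defining capsule bodies and joints for the pack and its bear's head, arms and torso bones.
# Constraints link root to pack to upper body, which branches to head, arms and lower body.
# Joints carry tight angular limits, and the arms lock twist and one swing axis, leaving a single limited swing.
# The upper-body bone is spelled "dyn_uppeeBody".
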PhysicsAsset /BRCosmetics/Accessories/FORT_Backpacks/Backpack_F_MED_Fuzzy_Bear_Summer/Meshes/F_MED_Fuzzy_Bear_Summer_Pack_Physics
bodies {
 dyn_Arms [{shapeType=Capsule, length=1, radius=1, shapeName=dyn_Arms_capsule}]
 dyn_head [{shapeType=Capsule, length=1, radius=1, shapeName=dyn_head_capsule, center=<0 0 2>, rotation=<rot 0 0 -8>}]
 dyn_lowerBody [{shapeType=Capsule, length=1, radius=1, shapeName=dyn_lowerBody_capsule, center=<0 0 -2>}]
 dyn_uppeeBody [{shapeType=Capsule, length=1, radius=1, shapeName=dyn_uppeeBody_capsule}]
 pack [{shapeType=Capsule, length=0.505, radius=0.505, shapeName=pack_capsule}]
 root [{shapeType=Capsule, length=0.505, radius=0.505, shapeName=root_capsule}]
}
constraints [
 {bone1Name=pack, bone2Name=root, swing1LimitDegrees=10, swing2LimitDegrees=10, twistLimitDegrees=10}
 {bone1Name=dyn_uppeeBody, bone2Name=pack, swing1LimitDegrees=5, swing2LimitDegrees=5, twistLimitDegrees=5}
 {bone1Name=dyn_head, bone2Name=dyn_uppeeBody, swing1LimitDegrees=1, swing2LimitDegrees=20, twistLimitDegrees=20}
 {
  bone1Name=dyn_Arms
  bone2Name=dyn_uppeeBody
  swing1LimitDegrees=10
  swing1Motion=Locked
  swing2LimitDegrees=10
  twistLimitDegrees=10
  twistMotion=Locked
 }
 {bone1Name=dyn_lowerBody, bone2Name=dyn_uppeeBody, swing1LimitDegrees=20, swing2LimitDegrees=20, twistLimitDegrees=20}
]
